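# Debug material for blinking text, unlit, translucent and two-sided, in an old landscape prototype folder.
# The red channel of a distance field font parameter (default RobotoDistanceField) drives a steep lerp between negative and positive extremes to sharpen glyph edges.
# That result multiplies by floored Time modulo two, switching it on and off every second.
# A separate multiply takes vertex color, presumably to tint the text.
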Material /Landmass/Landscape/OldPrototype_BP/Mats/Debug/BlinkingText
graph {
 properties {blendMode=BLEND_Translucent, pixelDepthOffsetMode=PDOM_Legacy, shadingModel=MSM_Unlit, twoSided=true, usedShadingModels=Unlit}
 nodes [
  n1 = Time()
  Multiply(
   A=LinearInterpolate(Alpha=FontSampleParameter() {font=/Engine/EngineFonts/RobotoDistanceField, parameterName=Font}.R) {constA=-90, constB=90}
   B=Fmod(A=Floor(None=@n1), B=Constant() {r=2})
  )
  Multiply(A=VertexColor())
 ]
}
