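# Unlit debug material in the old Landmass landscape brush prototype, apparently meant for copying a render target, with negative emissive values allowed.
# It holds an RT texture parameter that defaults to black.
# A separate DebugScalarValues function reads LandscapeZLocation from the MPC_Landscape parameter collection to display it as a number.
# Neither node is shown wired to a material output.
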Material /Landmass/Landscape/OldPrototype_BP/Mats/Debug/CopyRT
graph {
 properties {bAllowNegativeEmissiveColor=true, pixelDepthOffsetMode=PDOM_Legacy, shadingModel=MSM_Unlit, usedShadingModels=Unlit}
 nodes [
  MaterialFunctionCall(Number=CollectionParameter() {collection=/Landmass/Landscape/BlueprintBrushes/MPC/MPC_Landscape, parameterName=LandscapeZLocation}) {materialFunction=&DebugScalarValues}
  TextureSampleParameter2D() {parameterName=RT, texture=&Black}
 ]
}
