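# Unlit debug material in the Landmass landscape prototype that tests procedural noise through a Custom HLSL node.
# UVs, extended to a 3D position and multiplied by a scalar parameter, feed a direct call to the engine's MaterialExpressionNoise shader function.
# Parameters expose function, octaves and turbulence at runtime, while quality comes from a constant.
# A standard Noise node and a Quality parameter sit unwired, and no material output connection is shown.
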Material /Landmass/Landscape/OldPrototype_BP/Mats/Debug/CustomNoise
graph {
 properties {pixelDepthOffsetMode=PDOM_Legacy, shadingModel=MSM_Unlit, usedShadingModels=Unlit}
 nodes [
  MaterialExpressionMultiply_30 = Multiply(A=AppendVector(A=TextureCoordinate(), B=Constant()), B=ScalarParameter() {defaultValue=16, parameterName=fdfdf})
  MaterialExpressionConstant_8 = Constant() {r=4}
  MaterialExpressionScalarParameter_32 = ScalarParameter() {parameterName=F}
  MaterialExpressionScalarParameter_33 = ScalarParameter() {defaultValue=1, parameterName=Octaves}
  MaterialExpressionScalarParameter_34 = ScalarParameter() {parameterName=Turbulence}
  Custom!(
   InPos=@MaterialExpressionMultiply_30
   Quality=@MaterialExpressionConstant_8
   Function=@MaterialExpressionScalarParameter_32
   Octaves=@MaterialExpressionScalarParameter_33
   Turbulence=@MaterialExpressionScalarParameter_34
  ) {
   code="return MaterialExpressionNoise(InPos,1.00000000,Quality,Function,Turbulence,Octaves,0,1,2,0,0,512);"
   inputs [
    {inputName=InPos, input {expression=&self:MaterialExpressionMultiply_30}}
    {inputName=Quality, input {expression=&self:MaterialExpressionConstant_8}}
    {inputName=Function, input {expression=&self:MaterialExpressionScalarParameter_32}}
    {inputName=Octaves, input {expression=&self:MaterialExpressionScalarParameter_33}}
    {inputName=Turbulence, input {expression=&self:MaterialExpressionScalarParameter_34}}
   ]
  }
  Noise!(World Position, FilterWidth)
  ScalarParameter() {parameterName=Quality}
 ]
}
