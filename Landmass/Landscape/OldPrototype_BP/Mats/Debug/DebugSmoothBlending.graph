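# Unlit debug material that visualizes the Smooth Max output of the Landmass prototype's SmoothIntersections function as a graph.
# One PlotFunctionOnGraph call plots Smooth Max of the x ramp and its inverse (1 - x), with the Smoothing parameter clamped at zero as blend factor K.
# A second call plots 1 - x alone, and an Add combines both plots.
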Material /Landmass/Landscape/OldPrototype_BP/Mats/Debug/DebugSmoothBlending
graph {
 properties {pixelDepthOffsetMode=PDOM_Legacy, shadingModel=MSM_Unlit, usedShadingModels=Unlit}
 nodes [
  n1 = Constant4Vector() {constant=<rgba 0 1 0 1>}
  n2 = Constant3Vector() {constant=<rgba 1 0 0 0>}
  n3 = ScalarParameter() {defaultValue=-0.1369, parameterName=Smoothing}
  n4 = MaterialFunctionCall!(
   X range Y range=@n1.RGBA
   Gradient Height
   ShowGrid?
   Thickness
   Color=@n2
   "f(x)"=MaterialFunctionCall!(A=Multiply(A=@n4.x), B=@n5, K=Max(A=@n3) {constB=0}) {materialFunction=/Landmass/Landscape/OldPrototype_BP/Mats/MF/SmoothIntersections}.Smooth Max
  ) {materialFunction=/Engine/Functions/Engine_MaterialFunctions01/Debug/PlotFunctionOnGraph}
  n5 = OneMinus(None=@n4.x)
  Add(
   A=MaterialFunctionCall(X range Y range=@n1.RGBA, Color=@n2, "f(x)"=@n5) {materialFunction=&PlotFunctionOnGraph}."f(x) = y"
   B=@n4."f(x) = y"
  )
  Constant3Vector() {constant=<rgba 0 1 0 0>}
  TextureCoordinate()
 ]
}
