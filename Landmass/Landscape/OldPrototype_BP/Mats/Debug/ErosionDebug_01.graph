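# Unlit debug material in the old Landmass landscape prototype that visualises an erosion water-distribution test as five side-by-side columns.
# A Custom HLSL node ranks five parameter heights, then fills them level by level from a shared seeded water pool.
# Columns show black terrain and blue water, with Show blending from fixed to pooled water and DebugIntValues printing each rank.
# A second, simpler single-pass water-level Custom node sits unconnected.
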Material /Landmass/Landscape/OldPrototype_BP/Mats/Debug/ErosionDebug_01
graph {
 properties {bAllowNegativeEmissiveColor=true, pixelDepthOffsetMode=PDOM_Legacy, shadingModel=MSM_Unlit, usedShadingModels=Unlit}
 nodes [
  n1 = TextureSample(UVs=TextureCoordinate() {uTiling=5, vTiling=5}) {samplerType=SAMPLERTYPE_Masks, texture=/Engine/EngineMaterials/DefaultDiffuse_TC_Masks}
  n2 = Multiply(A=Constant3Vector() {constant=<rgba 0 0 1 0>}, B=@n1.RGB)
  n3 = OneMinus(None=ComponentMask(None=TextureCoordinate()) {g=true})
  n4 = TextureCoordinate()
  MaterialExpressionFloor_0 = Floor(None=Multiply(A=ComponentMask(None=@n4) {r=true}) {constB=5})
  MaterialExpressionScalarParameter_3 = ScalarParameter() {defaultValue=0.01, parameterName=a, sliderMax=1}
  MaterialExpressionScalarParameter_4 = ScalarParameter() {defaultValue=0.1, parameterName=b, sliderMax=1}
  MaterialExpressionScalarParameter_5 = ScalarParameter() {defaultValue=0.2, parameterName=c, sliderMax=1}
  MaterialExpressionScalarParameter_6 = ScalarParameter() {defaultValue=0.3, parameterName=d, sliderMax=1}
  MaterialExpressionScalarParameter_7 = ScalarParameter() {defaultValue=0.4, parameterName=e, sliderMax=1}
  MaterialExpressionScalarParameter_8 = ScalarParameter() {defaultValue=5, parameterName=MaxLevels}
  MaterialExpressionScalarParameter_1 = ScalarParameter() {defaultValue=0.1, parameterName=fixedwater}
  n5 = Custom!(
   x=@MaterialExpressionFloor_0
   a=@MaterialExpressionScalarParameter_3
   b=@MaterialExpressionScalarParameter_4
   c=@MaterialExpressionScalarParameter_5
   d=@MaterialExpressionScalarParameter_6
   e=@MaterialExpressionScalarParameter_7
   maxlevel=@MaterialExpressionScalarParameter_8
   seedwater=@MaterialExpressionScalarParameter_1
  ) {
   code="float2 Offsets[5] = {float2(0,1),float2(-1,0), float2(0,0),float2(1,0),float2(0,1)};\r\n\r\nfloat heights[5] = {0.1,0.2,0.3,0.4,0.5};\r\nfloat water[5] =   {0,0,0,0,0};\r\nint orders[5] =    {0,0,0,0,0};\r\nint actives[5] =   {0,0,0,0,0};\r\n\r\n//float texelsize = 1 / TextureSize;\r\nfloat watersum = seedwater*5;\r\n\r\nheights[0] = a;\r\nheights[1] = b;\r\nheights[2] = c;\r\nheights[3] = d;\r\nheights[4] = e;\r\n\r\n//Sample Heightmap and water data\r\nfor (int i = 0; i < 5; i++)\r\n{\r\n\t//heights[i] = HeightTex.SampleLevel(HeightTexSampler, UV + offsets[i] * texelsize, 0);\r\n\t//watersum += WaterTex.SampleLevel(WaterTexSampler, UV + offsets[i] * texelsize, 0);\r\n}\r\n\r\n//sort heights by how many are lower than each\r\nfloat lowesttemp = 65000;\r\nfor (int j = 0; j < 5; j++)\r\n{\r\n\tfloat curheight = heights[j];\r\n\r\n\tfor (int k = 0; k < 5; k++)\r\n\t{\r\n\t\tfloat checkheight = heights[k];\r\n\t\tif (checkheight < curheight)\r\n\t\t{\r\n\t\t\torders[j] += 1;\r\n\t\t}\r\n\t}\r\n}\r\n\r\n//use sort data to iterate again\r\nfor (int l = 0; l < min(maxlevel,5); l++)\r\n{\r\n\r\n\tfloat curactivecells = 0;\r\n\tfloat curpooltop = 1;\r\n\tfloat curpoolbottom = 0;\r\n\t\r\n\tfor (int m = 0; m < 5; m++)\r\n\t{\r\n\t\tif(orders[m] <= l)\r\n\t\t{\r\n\t\t\tcuractivecells += 1;\r\n\t\t\tcurpoolbottom = max(curpoolbottom,heights[m]+water[m]);\r\n\t\t\tactives[m] = 1;\r\n\t\t}\r\n\t\telse\r\n\t\t{\r\n\t\t\tactives[m] = 0;\r\n\t\t}\r\n\t\tif(orders[m] > l )\r\n\t\t{\r\n\t\t\tcurpooltop = min(curpooltop,heights[m]);\r\n\t\t}\r\n\t\t\r\n\t}\r\n\r\n\tfloat curpoolsize = (curpooltop - curpoolbottom) * curactivecells;\r\n\tcurpoolsize = min(curpoolsize, watersum);\r\n\tcurpoolsize = max(0, curpoolsize);\r\n\r\n\tfor (int n = 0; n < 5; n++)\r\n\t{\r\n\t\tif(actives[n] == 1)\r\n\t\t{\r\n\t\t\twater[n] += curpoolsize / curactivecells;\r\n\t\t\twatersum -= curpoolsize / curactivecells;\r\n\t\t}\r\n\t}\r\n\t\r\n\r\n}\r\n\r\nint index = min(4, x);\r\n\r\nreturn float3(heights[x],water[index],orders[index]);"
   inputs [
    {inputName=x, input {expression=&self:MaterialExpressionFloor_0}}
    {inputName=a, input {expression=&self:MaterialExpressionScalarParameter_3}}
    {inputName=b, input {expression=&self:MaterialExpressionScalarParameter_4}}
    {inputName=c, input {expression=&self:MaterialExpressionScalarParameter_5}}
    {inputName=d, input {expression=&self:MaterialExpressionScalarParameter_6}}
    {inputName=e, input {expression=&self:MaterialExpressionScalarParameter_7}}
    {inputName=maxlevel, input {expression=&self:MaterialExpressionScalarParameter_8}}
    {inputName=seedwater, input {expression=&self:MaterialExpressionScalarParameter_1}}
   ]
  }
  n6 = ComponentMask(None=@n5) {r=true}
  n7 = Add(
   A=@n6
   B=LinearInterpolate(A=@MaterialExpressionScalarParameter_1, B=ComponentMask(None=@n5) {g=true}, Alpha=ScalarParameter() {parameterName=Show, sliderMax=1})
  )
  n8 = Saturate(None=Ceil(None=Subtract(A=@n3, B=@n7)))
  n9 = Ceil(None=Subtract(A=@n3, B=@n6))
  n10 = ComponentMask(None=@n5) {b=true}
  n11 = Frac(None=Multiply(A=TextureCoordinate()) {constB=5})
  n12 = TextureCoordinate()
  n13 = Ceil(None=Subtract(A=ComponentMask(None=@n12) {g=true}) {constB=0.75})
  Add(
   A=LinearInterpolate(B=LinearInterpolate(A=@n2, B=@n1.RGB, Alpha=@n8), Alpha=Saturate(None=@n9))
   B=Multiply(
    A=MaterialFunctionCall!(Number=@n10, MaximumNumberOfDigits, DebugTextLocation RG_UpperRight BA_LowerLeft, UVs=@n11) {materialFunction=/Engine/Functions/Engine_MaterialFunctions01/Debug/DebugIntValues}.Result
    B=Saturate(None=@n13)
   )
  )
  Custom(
   x=@MaterialExpressionFloor_0
   a=@MaterialExpressionScalarParameter_3
   b=@MaterialExpressionScalarParameter_4
   c=@MaterialExpressionScalarParameter_5
   d=@MaterialExpressionScalarParameter_6
   e=@MaterialExpressionScalarParameter_7
   maxlevel=@MaterialExpressionScalarParameter_8
   seedwater=@MaterialExpressionScalarParameter_1
  ) {
   code="//float2 Offsets[5] = {float2(0,1),float2(-1,0), float2(0,0),float2(1,0),float2(0,1)};\r\n\r\nfloat heights[5] = {a,b,c,d,e};\r\nfloat water[5] =   {0,0,0,0,0};\r\n\r\n//float texelsize = 1 / TextureSize;\r\nfloat watersum = seedwater*5;\r\n\r\n//heights[0] = a;\r\n//heights[1] = b;\r\n//heights[2] = c;\r\n//heights[3] = d;\r\n//heights[4] = e;\r\n\r\n//Sample Heightmap and water data\r\n//for (int i = 0; i < 5; i++)\r\n//{\r\n\t//heights[i] = HeightTex.SampleLevel(HeightTexSampler, UV + offsets[i] * texelsize, 0);\r\n\t//watersum += WaterTex.SampleLevel(WaterTexSampler, UV + offsets[i] * texelsize, 0);\r\n//}\r\n\r\nfloat altsum = heights[0];\r\nfloat cells = 1;\r\nfloat waterlevel = heights[0] + watersum;\r\n\r\nfor (int j = 1; j < 5; j++)\r\n{\r\n\r\n\t//if( heights[j] < waterlevel )\r\n\t//{\r\n\t\tfloat mask = heights[j] < waterlevel ? 1 : 0;\r\n\t\tcells += 1 * mask;\r\n\t\taltsum += heights[j] * mask;\r\n\t\twaterlevel = ( altsum / cells ) + (watersum/cells);\r\n\t//}\r\n}\r\n\r\n\r\nint index = min(4, x);\r\nreturn float3(heights[index],max(0, waterlevel-heights[index]),0);"
   inputs [
    {inputName=x, input {expression=&self:MaterialExpressionFloor_0}}
    {inputName=a, input {expression=&self:MaterialExpressionScalarParameter_3}}
    {inputName=b, input {expression=&self:MaterialExpressionScalarParameter_4}}
    {inputName=c, input {expression=&self:MaterialExpressionScalarParameter_5}}
    {inputName=d, input {expression=&self:MaterialExpressionScalarParameter_6}}
    {inputName=e, input {expression=&self:MaterialExpressionScalarParameter_7}}
    {inputName=maxlevel, input {expression=&self:MaterialExpressionScalarParameter_8}}
    {inputName=seedwater, input {expression=&self:MaterialExpressionScalarParameter_1}}
   ]
  }
  n14 = Divide(A=@MaterialExpressionFloor_0) {constB=5}
  n15 = ScalarParameter() {defaultValue=0.5}
  Saturate(None=Add(A=Multiply(A=@n14, B=@n15)) {constB=0.1})
 ]
}
