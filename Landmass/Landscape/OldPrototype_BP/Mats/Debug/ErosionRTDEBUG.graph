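# Unlit debug material that draws an erosion render target (RT) as a cross-section chart, one column per texel across SizeX, on a 64-unit vertical scale.
# R fills black, G stacks above it in textured blue over a tiled background, and B is a dark blue overlay blended by op.
# A Custom HLSL loop sums G across up to 256 texels for a DebugScalarValues readout.
# A per-column bottom readout subtracts B from itself, so its value is always zero.
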Material /Landmass/Landscape/OldPrototype_BP/Mats/Debug/ErosionRTDEBUG
graph {
 properties {bAllowNegativeEmissiveColor=true, pixelDepthOffsetMode=PDOM_Legacy, shadingModel=MSM_Unlit, usedShadingModels=Unlit}
 nodes [
  n1 = Multiply(A=TextureSample(UVs=TextureCoordinate() {uTiling=5, vTiling=5}) {samplerType=SAMPLERTYPE_Masks, texture=/Engine/EngineMaterials/DefaultDiffuse_TC_Masks}.RGB) {constB=0.3}
  n2 = TextureCoordinate()
  n3 = Multiply(A=OneMinus(None=ComponentMask(None=@n2) {g=true})) {constB=64}
  n4 = TextureCoordinate()
  n5 = ScalarParameter() {defaultValue=5, parameterName=SizeX}
  n6 = Floor(None=Multiply(A=ComponentMask(None=@n4) {r=true}, B=@n5) {constB=5})
  n7 = TextureSampleParameter2D(UVs=Add(A=Divide(A=@n6, B=@n5) {constB=5}, B=Divide(B=@n5) {constA=0.5, constB=5})) {parameterName=RT, texture=&Black}
  n8 = ComponentMask(None=@n7.R) {r=true}
  n9 = Subtract(A=@n3, B=Add(A=@n8, B=ComponentMask(None=@n7.G) {g=true}))
  n10 = LinearInterpolate(A=Multiply(A=Constant3Vector() {constant=<rgba 0 0 1 0>}, B=@n1), B=@n1, Alpha=Saturate(None=Ceil(None=@n9)))
  n11 = Saturate(None=Ceil(None=Subtract(A=@n3, B=@n8)))
  n12 = Subtract(A=@n3, B=@n7.B)
  n13 = OneMinus(None=Saturate(None=Ceil(None=@n12)))
  n14 = ScalarParameter() {parameterName=op}
  MaterialExpressionTextureObjectParameter_0 = TextureObjectParameter() {parameterName=RT, texture=&Black}
  MaterialExpressionScalarParameter_4 = ScalarParameter() {defaultValue=5, parameterName=SizeX}
  n15 = Custom!(Tex=@MaterialExpressionTextureObjectParameter_0, SizeX=@MaterialExpressionScalarParameter_4) {
   code="float watersum = 0;\r\n\r\nfloat2 texelsize = 1 / float2(SizeX, 1);\r\n\r\nfloat2 UV = texelsize / 2;\r\n\r\nfor (int i = 0; i < min(256, SizeX); i++)\r\n{\r\n\twatersum += Tex.SampleLevel(TexSampler, UV + float2( texelsize.x * i, 0.5), 0).g;\r\n}\r\n\r\nreturn watersum;"
   outputType=CMOT_Float1
   inputs [
    {inputName=Tex, input {expression=&self:MaterialExpressionTextureObjectParameter_0}}
    {inputName=SizeX, input {expression=&self:MaterialExpressionScalarParameter_4}}
   ]
  }
  n16 = TextureCoordinate()
  n17 = ScalarParameter() {defaultValue=5, parameterName=SizeX}
  n18 = MaterialFunctionCall(Number=Max(A=Subtract(A=@n7.B, B=@n7.B)) {constB=0}, UVs=Frac(None=Multiply(A=@n16, B=@n17) {constB=5})) {materialFunction=&DebugScalarValues}
  n19 = ComponentMask(None=TextureCoordinate()) {g=true}
  n20 = ScalarParameter() {defaultValue=0.9, parameterName=asdf}
  n21 = Saturate(None=Ceil(None=Subtract(A=@n19, B=@n20) {constB=0.75}))
  Add(
   A=LinearInterpolate(A=LinearInterpolate(B=@n10, Alpha=@n11), B=Constant3Vector() {constant=<rgba 0 0.033 0.07 0>}, Alpha=Multiply(A=@n13, B=@n14))
   B=Add(
    A=MaterialFunctionCall(Number=@n15) {materialFunction=&DebugScalarValues}.Result
    B=Multiply(A=@n18.Result, B=@n21)
   )
  )
 ]
}
