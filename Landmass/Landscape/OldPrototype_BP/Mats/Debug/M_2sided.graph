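# M_2sided is a two-sided unlit debug material in the Landmass old landscape Blueprint prototype folder.
# Its graph contains no expressions, so every output stays at its default and surfaces render flat black from both sides.
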Material /Landmass/Landscape/OldPrototype_BP/Mats/Debug/M_2sided
graph {expressions {}, properties {pixelDepthOffsetMode=PDOM_Legacy, shadingModel=MSM_Unlit, twoSided=true, usedShadingModels=Unlit}}
