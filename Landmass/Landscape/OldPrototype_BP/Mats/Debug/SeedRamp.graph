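# SeedRamp is an unlit debug material from an old landscape prototype, apparently sketching a noise ramp from a seed-like coordinate.
# The U texture coordinate is scaled by parameter H, then feeds a Floor, a MakeFloat3 alongside parameter Water, and a Noise position input.
# The non-turbulent noise is multiplied by parameter M.
# No connection to a material output appears, and a Divide node sits unwired, so the graph looks unfinished.
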Material /Landmass/Landscape/OldPrototype_BP/Mats/Debug/SeedRamp
graph {
 properties {pixelDepthOffsetMode=PDOM_Legacy, shadingModel=MSM_Unlit, usedShadingModels=Unlit}
 nodes [
  Divide()
  n1 = Multiply(A=ComponentMask(None=TextureCoordinate()) {r=true}, B=ScalarParameter() {defaultValue=64, parameterName=H})
  Floor(None=@n1)
  MaterialFunctionCall(X=@n1, Y=ScalarParameter() {defaultValue=1, parameterName=Water}) {materialFunction=&MakeFloat3}
  Multiply(A=Noise!(World Position=@n1, FilterWidth) {bTurbulence=false, outputMin=0}, B=ScalarParameter() {defaultValue=32, parameterName=M})
 ]
}
