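# TickWarning is an unlit masked debug material in the old Landmass prototype that draws a blinking red warning indicator.
# The Sprite function centers camera-facing geometry on a point just ahead of the camera along its view direction, scaled by parameter X.
# The Sequence Recorder's recording-indicator alpha is multiplied by floored time modulo two, so the icon toggles on and off each second.
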Material /Landmass/Landscape/OldPrototype_BP/Mats/Debug/TickWarning
graph {
 properties {blendMode=BLEND_Masked, pixelDepthOffsetMode=PDOM_Legacy, shadingModel=MSM_Unlit, usedShadingModels=Unlit}
 nodes [
  Constant3Vector() {constant=<rgba 2 0 0 0>}
  n1 = MaterialFunctionCall() {materialFunction=&CameraDirectionVector}
  MaterialFunctionCall!(
   Center Location=Add(A=CameraPositionWS(), B=Multiply(A=@n1.Result) {constB=100})
   0-1 UVs=TextureCoordinate()
   XY scale=ScalarParameter() {defaultValue=10, parameterName=X}
   OverridePivot
   "Normals (Optional)"
  ) {materialFunction=&Sprite}
  n2 = Time()
  Multiply(A=TextureSample() {texture=/Engine/EditorResources/SequenceRecorder/RecordingIndicator}.A, B=Fmod(A=Floor(None=@n2), B=Constant() {r=2}))
 ]
}
